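# Niagara GPU effect that scatters bamboo leaves on landscape terrain and stirs them around the camera.
# A short spawn burst fills a user-set radius, with landscape sampling supplying terrain height and normal.
# A pulsing, negative-strength attractor offset from the camera pushes leaves, activating drag, wind and camera-facing orientation alongside gravity and aerodynamic drag.
# Particle aging depends on velocity, and distance culling and scaling limit the effect.
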NiagaraSystem /Game/Effects/Fort_Effects/Effects/Environment/Forest/NS_Env_BambooLeaves_OnGround_Interact_LandScape
dependencies {
 usedDataInterfaces [&NiagaraDataInterfaceCamera, &NiagaraDataInterfaceCurve, &NiagaraDataInterfaceCollisionQuery]
 usedDynamicInputs []
 usedModules [&SystemState]
}
emitters {Bamboo_Interactive_LandScape {
  simTarget=GPUComputeSim
  emitterUpdateScript [
   /SetVariables() {isSetParameters=true}
   /SetVariables(
    Emitter.CameraForward:Vector3f=&ReturnCameraVector(Camera Property=::NewEnumerator1, Coordinate Space=::Simulation, CameraQuery=$Emitter.CameraRead)
    Emitter.CameraPosition:Vector3f=&ReturnCameraVector(Camera Property=::NewEnumerator0, Coordinate Space=::World, CameraQuery=$Emitter.CameraRead)
    Emitter.CameraRight:Vector3f=&ReturnCameraVector(Camera Property=::NewEnumerator3, Coordinate Space=::Simulation, CameraQuery=$Emitter.CameraRead)
    Emitter.CameraUp:Vector3f=&ReturnCameraVector(Camera Property=::NewEnumerator2, Coordinate Space=::Simulation, CameraQuery=$Emitter.CameraRead)
   ) {isSetParameters=true}
   &EmitterState(
    Enable Distance Culling=true
    Life Cycle Mode=::NewEnumerator1
    Loop Behavior=::NewEnumerator1
    Loop Duration Mode=::NewEnumerator1
    Max Distance Response=::SleepAndClearParticles
    MaxDistance=8000
    Module.bUseMaxDistance=true
    Module.bUseMinDistance=false
    Reset Age On Awaken=true
    Scalability Mode=::NewEnumerator1
    UseLoopDelay=false
   )
   &SpawnRate(
    SpawnRate=&FloatFromCurve(FloatCurve=DataInterface() {propertyValues {Curve {keys [<at 0 0>,<at 0.1 1>,<at 0.2 0>]}}}, CurveIndex=$Emitter.Age, Scale Curve=15000)
    Use Spawn Probability=false
   )
  ]
  particleSpawnScript [
   &InitializeParticle(
    Color Maximum=<rgba 0.648 0.648 0.648 1>
    Color Minimum=<rgba 0.2827 0.2827 0.2827 1>
    Color Mode=::NewEnumerator2
    Lifetime=&RandomRangeFloat(Minimum=3, Maximum=5)
    Mass=3
    Mass Mode=::NewEnumerator1
    Mesh Scale Mode=::NewEnumerator1
    Mesh Uniform Scale Max=4
    Mesh Uniform Scale Min=2
    Sprite Rotation Mode=::NewEnumerator2
    Sprite Size Max=<50 140>
    Sprite Size Min=<40 120>
    Sprite Size Mode=::NewEnumerator2
    UsePositionOffset=false
    Write Material Random=true
   ) {module=InitializeParticle001}
   /SetVariables() {isSetParameters=true}
   &InitialMeshOrientation(
    Mesh Orientation Mode=::NewEnumerator1
    Rotation=&RandomRangeVector(Minimum=<0 0 0>, Maximum=<0 0 1>)
    Rotation Coordinate Space=::NewEnumerator2
    Use Rotation Vector=true
   )
   &ShapeLocation(
    MaskSpawnGroup=false
    Offset=&RandomRangeVector(Minimum=&Multiply_VectorByFloat(Vector=$User.RadiusOffset, Float=-1), Maximum=$User.RadiusOffset)
    Offset Mode=::NewEnumerator0
    Override Seed=false
    Sphere Radius=$User.Radius
    UseOwnerScale=true
   )
   &self:Sample_Landscape_01() {module=Sample_Landscape_01}
   /SetVariables(Particles.Position=&Add_Vector(
     A=Linked() {linkedVariable {name=Particles.Position, type=NiagaraPosition}}
     B=&MakeVector(X=0, Y=0, Z=$Particles.Terrain Offset)
    )) {isSetParameters=true}
   /SetVariables(Particles.New Vector 3f:Vector3f=<0 0 1>) {isSetParameters=true}
   &CameraOffset(Camera Offset Amount=5)
   /SetVariables(Particles.SubImageIndex=&RandomRangeFloat(Minimum=0, Maximum=63)) {isSetParameters=true}
   /SetVariables() {isSetParameters=true}
   /SetVariables(Particles.ColorParticle:LinearColor=&RandomRangeLinearColor(Minimum=<rgba 0.2784 0.2784 0.2784 1>, Maximum=<rgba 1 0.9278 0.6564 1>)) {isSetParameters=true}
  ]
  particleUpdateScript [
   &self:Sample_Landscape_UPdate_01(Bias:NiagaraFloat=$Particles.Terrain Offset, New Landscape Sample:NiagaraDataInterfaceLandscape=$User.Landscape) {module=Sample_Landscape_UPdate_01}
   &ScaleColor(
    Scale Alpha=&FloatFromCurve(FloatCurve=DataInterface() {propertyValues {Curve {keys [<at 0.8 1>,<at 1 0>]}}}, Scale Curve=5)
    Scale RGB=&MakeVectorFromLinearColorRGB(Color=&MultiplyLinearColor(Linear Color=$Particles.Initial.Color, Scale Factor=&MakeVector4_FromFloat(Float=1)))
    ScaleA=true
    ScaleRGB=true
   )
   /SetVariables(Particles.Has Collided With Terrain:NiagaraBool=&SetBoolByFloatComparison(
     A=&Add_Float(A=$Output.Sample_Landscape_UPdate_01.Terrain Height, B=5)
     B=&MakeFloatFromVector(Vector=Linked() {linkedVariable {name=Particles.Position, type=NiagaraPosition}}, Channel=::NewEnumerator2)
     Comparison Type=::NewEnumerator2
    )) {isSetParameters=true}
   /Niagara/Modules/Update/Forces/PointAttractionForce(
    Attraction Radius:NiagaraFloat=&Waveform(
     Waveform Count=::NewEnumerator1
     Global Amplitude Scale=65
     "Waveform [1]"=::NewEnumerator0
     "[1] Frequency"=5
     "[1] Amplitude Min/Max"=<0.4 1>
     "[2] Average"=false
     "[2] Input Position"=0
     "[2] Phase"=$Emitter.Age
     "[2] Phase Scale"=1
     "[2] Compound Frequency"=2
     "[2] Frequency"=2
     "[2] Amplitude Min/Max"=<0 1>
     "[2] Scale"=1
     "[2] Bias"=0
     "[2] Exponent"=1
     "[2] Pulse Duration"=0.1
     "[2] Clip Triangle Min/Max"=<0 0>
     "[2] Random Seed"=0
    )
    AttractionStrength:NiagaraFloat=&Multiply_Float(
     A=&FloatFromCurve(FloatCurve=DataInterface() {propertyValues {Curve {keys [<at 0 1>,<at 0.25 0>]}}}, Scale Curve=-65)
     B=1
    )
    AttractorPosition:Vector3f=&Add_Vector(
     A=$Emitter.CameraPosition
     B=&Add_Vector(
      A=&Add_Vector(
       A=&Multiply_VectorByFloat(Vector=$Particles.New Vector 3f, Float=-130)
       B=&Multiply_VectorByFloat(Vector=$Emitter.CameraForward, Float=250)
      )
      B=&Multiply_VectorByFloat(Vector=$Emitter.CameraRight, Float=-50)
     )
    )
    Falloff Exponent=1
    Module.Kill Within Radius=false
    Module.Use Falloff=false
   ) {module=PointAttractionForce002}
   /SetVariables(Particles.Normalized Velocity:Vector3f=&NormalizeVector(Vector To Normalize=$Particles.Velocity)) {isSetParameters=true}
   /SetVariables(Particles.Is Active:NiagaraBool=&Boolean_OrOperation(
     Bool1=$Particles.Is Active
     Bool2=&SetBoolByFloatComparison(
      A=&MakeFloatFromVector(Vector=Linked() {linkedVariable {name=Particles.Position, type=NiagaraPosition}}, Channel=::NewEnumerator2)
      B=$Output.Sample_Landscape_UPdate_01.Terrain Height
      Comparison Type=::NewEnumerator6
     )
    )) {isSetParameters=true}
   &Drag(
    Drag=&MakeCustomFloatFromBool(
     Bool=$Particles.Is Active
     True Float=&FloatFromCurve(
      FloatCurve=DataInterface() {propertyValues {Curve {keys [
          {interpMode=RCIM_Cubic, tangentMode=RCTM_User, time=0, value=0}
          {interpMode=RCIM_Cubic, tangentMode=RCTM_User, time=0.2, value=0.95}
          {interpMode=RCIM_Cubic, tangentMode=RCTM_User, time=0.5, value=0}
         ]}}}
      Scale Curve=0.25
     )
     False Float=0
    )
    Use Linear Drag=true
    Use Rotational Drag=false
   )
   &GravityForce(Gravity=&MakeVector(
     X=0
     Y=&FloatFromCurve(
      FloatCurve=DataInterface() {propertyValues {Curve {keys [
          <at 0 0>
          {interpMode=RCIM_Cubic, tangentMode=RCTM_User, time=0.02, value=2}
          <at 0.1 0>
          {interpMode=RCIM_Cubic, tangentMode=RCTM_User, time=0.2, value=-0.9}
          <at 0.5 -0.25>
          <at 0.8 0.1>
          <at 1 0.1>
         ]}}}
      Scale Curve=&RandomRangeFloat(Minimum=-250, Maximum=250)
     )
     Z=&FloatFromCurve(
      FloatCurve=DataInterface() {propertyValues {Curve {keys [
          <at 0 0>
          {interpMode=RCIM_Cubic, tangentMode=RCTM_User, time=0.02, value=2}
          <at 0.1 0>
          {interpMode=RCIM_Cubic, tangentMode=RCTM_User, time=0.2, value=-0.9}
          <at 0.5 -0.25>
          <at 0.8 0.1>
          <at 1 0.1>
         ]}}}
      Scale Curve=&RandomRangeFloat(Minimum=180, Maximum=250)
     )
    ))
   /SetVariables(Particles.Sprite Orientation:Vector3f=&Lerp_Vector(
     A=$Particles.Landscape Normal
     B=&ReturnCameraVector(Camera Property=::NewEnumerator4, Coordinate Space=::Simulation, Origin Position=$Particles.Position, CameraQuery=$Emitter.CameraRead)
     Alpha=&MakeCustomFloatFromBool(Bool=$Particles.Is Active, True Float=1, False Float=0)
    )) {isSetParameters=true}
   &ParticleState(DeltaTime:NiagaraFloat=&MakeCustomFloatFromBool(
     Bool=&SetBoolByFloatComparison(A=&VectorLength(Vector=$Particles.Velocity), B=1)
     True Float=$Engine.DeltaTime
     False Float=0
    ))
   &self:ScratchModule_05(Terrain Height:NiagaraFloat=$Output.Sample_Landscape_UPdate_01.Terrain Height, Threshold=50) {module=ScratchModule_05}
   &ScaleMeshSize(Scale Factor=&VectorFromFloat(Value=&Multiply_Float(
      A=&FloatFromCurve(
       FloatCurve=DataInterface() {propertyValues {Curve {keys [<at 0.25 0>,<at 500 1>,<at 6000 1>,<at 8000 0>]}}}
       CurveIndex=$Engine.Owner.LODDistance
       Scale Curve=0.33
      )
      B=&FloatFromCurve(FloatCurve=DataInterface() {propertyValues {Curve {keys [<at 0.8 1>,<at 1 0>]}}}, Scale Curve=0.25)
     )))
   &ScaleSpriteSize(
    Scale Sprite Size Mode=::NewEnumerator0
    Uniform Scale Factor=&Multiply_Float(
     A=&FloatFromCurve(
      FloatCurve=DataInterface() {propertyValues {Curve {keys [<at 0 1>,<at 500 1>,<at 6000 1>,<at 8000 0>]}}}
      CurveIndex=$Engine.Owner.LODDistance
     )
     B=&FloatFromCurve(FloatCurve=DataInterface() {propertyValues {Curve {keys [<at 0.8 1>,<at 1 0>]}}}, Scale Curve=0.25)
    )
   )
   &WindForce(
    Offset By Time:Vector3f=<0 0 -10>
    Scale=1
    Use Ground Mask=true
    Wind Speed=&MakeVector(
     X=&MakeCustomFloatFromBool(
      Bool=$Particles.Is Active
      True Float=&FloatFromCurve(FloatCurve=DataInterface() {propertyValues {Curve {keys [<at 0.25 1>,<at 0.5 0.25>]}}}, Scale Curve=120)
      False Float=0
     )
     Y=0
     Z=0
    )
   )
   &AerodynamicDrag(
    Aerodynamic Drag=0.1
    Aerodynamic Rotational Drag=0.1
    Drag Shape Curve=DataInterface() {propertyValues {Curve {keys [
        {interpMode=RCIM_Cubic, tangentMode=RCTM_User, time=0, value=0.1}
        {arriveTangent=2.0799, interpMode=RCIM_Cubic, leaveTangent=2.0799, tangentMode=RCTM_User, time=1, value=1}
       ]}}}
    Lift Shape Curve=DataInterface() {propertyValues {Curve {keys [
        {interpMode=RCIM_Cubic, tangentMode=RCTM_User, time=0, value=0}
        {interpMode=RCIM_Cubic, tangentMode=RCTM_User, time=0.5, value=1}
        {tangentMode=RCTM_User, time=1, value=0}
       ]}}}
    Pivot Offset Maximum:Vector3f=<1 1 1>
    Pivot Offset Minimum Length:Vector3f=<250 0 500>
    Rotation Strength:NiagaraFloat=&FloatFromCurve(
     FloatCurve=DataInterface() {propertyValues {Curve {keys [{interpMode=RCIM_Cubic, tangentMode=RCTM_User, time=0, value=1}, {interpMode=RCIM_Cubic, tangentMode=RCTM_User, time=1, value=0}]}}}
     Scale Curve=&RandomRangeFloat(Minimum=1, Maximum=4)
    )
    Use Linear Drag=true
    Use Rotational Drag=true
   ) {module=AerodynamicDrag002}
   &SolveForcesAndVelocity(Clamp Velocity=true, Limit Acceleration=false) {module=SolveForcesAndVelocity001}
   &AlignSpriteToMeshOrientation(
    Mesh Orientation Relative Sprite Alignment Vector=<1 0 0>
    Mesh Orientation Relative Sprite Facing Vector:Vector3f=<1 1 1>
    SetAlignment=true
    SetFacing=true
   ) {enabled=false}
   &AlignSpriteToMeshOrientation(Mesh Orientation Relative Sprite Alignment Vector=<0 1 1>, SetAlignment=true, SetFacing=true) {module=AlignSpriteToMeshOrientation001}
   &DynamicMaterialParameters(
    Index 0 Param 1=&RandomRangeFloat(Minimum=0.1, Maximum=0.15)
    Index 0 Param 2=&RandomRangeFloat(Minimum=0, Maximum=0)
    Module.Param0WriteEnabled=true
    Module.Param1WriteEnabled=true
    Module.Param2WriteEnabled=true
    Module.Param3WriteEnabled=true
   )
   &KillParticles(Kill Particles=&self:ScratchDynamicInput())
  ]
  rendererClasses [&NiagaraSpriteRendererProperties, &NiagaraMeshRendererProperties]
  renderers [
   {rendererClass=&NiagaraSpriteRendererProperties}
   {rendererClass=&NiagaraMeshRendererProperties, rendererIndex=1}
  ]
 }}
summary {
 systemName=NS_Env_BambooLeaves_OnGround_Interact_LandScape
 emitters [{
   emitterName=Bamboo_Interactive_LandScape
   simTarget=GPUComputeSim
   rendererClasses [&NiagaraSpriteRendererProperties, &NiagaraMeshRendererProperties]
  }]
 userVariables [
  User.Alpha_AttractionStrength {type=NiagaraFloat, defaultValue=0, description=""}
  User.Landscape {
   type=NiagaraDataInterfaceLandscape
   description=""
   defaultValue {
    struct=NiagaraExt_VariableValue_DataInterface
    value {
     dataInterface=&self:NiagaraDataInterfaceLandscape_0
     dataInterfaceClass=&NiagaraDataInterfaceLandscape
    }
   }
  }
  User.Player Position {type=Vector3f, description="", defaultValue {struct=Vector3f, value=<0 0 0>}}
  User.Radius {type=NiagaraFloat, defaultValue=300, description=""}
  User.RadiusOffset {type=Vector3f, description="", defaultValue {struct=Vector3f, value=<250 250 0>}}
 ]
}
