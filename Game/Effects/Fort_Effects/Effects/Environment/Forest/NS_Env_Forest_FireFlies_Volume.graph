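# Forest firefly volume: glowing, flickering green GPU sprites in a scalable sphere, faded in spawn rate and alpha by time of day to favor night.
# Fireflies bursts on awakening, snaps particles to depth-buffer surfaces with GBuffer color filtering, and holds them until motion activates noise and lift.
# Fireflies001 adds short-lived looping pulses that drift on curl noise.
# Both use a camera-relative repulsion point, distance culling and distance-based size scaling.
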NiagaraSystem /Game/Effects/Fort_Effects/Effects/Environment/Forest/NS_Env_Forest_FireFlies_Volume
dependencies {
 usedDataInterfaces [&NiagaraDataInterfaceCamera, &NiagaraDataInterfaceCurve, &NiagaraDataInterfaceGBuffer]
 usedDynamicInputs []
 usedModules [
  &SystemState
  &self:SystemSpawnScript.SystemScriptSource.SystemScriptGraph.NiagaraNodeAssignment_0.SetVariables_#1
 ]
}
emitters {
 Fireflies {
  simTarget=GPUComputeSim
  emitterUpdateScript [
   /SetVariables() {isSetParameters=true}
   /SetVariables(
    Emitter.CameraForward:Vector3f=&ReturnCameraVector(Camera Property=::NewEnumerator1, Coordinate Space=::Simulation, CameraQuery=$Emitter.CameraRead)
    Emitter.CameraPosition:Vector3f=&ReturnCameraVector(Camera Property=::NewEnumerator0, Coordinate Space=::World, CameraQuery=$Emitter.CameraRead)
    Emitter.CameraRight:Vector3f=&ReturnCameraVector(Camera Property=::NewEnumerator3, Coordinate Space=::Simulation, CameraQuery=$Emitter.CameraRead)
    Emitter.CameraUp:Vector3f=&ReturnCameraVector(Camera Property=::NewEnumerator2, Coordinate Space=::Simulation, CameraQuery=$Emitter.CameraRead)
   ) {isSetParameters=true}
   &EmitterState(
    Enable Distance Culling=true
    Life Cycle Mode=::NewEnumerator1
    Loop Behavior=::NewEnumerator1
    Loop Duration Mode=::NewEnumerator1
    Max Distance Response=::SleepAndClearParticles
    Module.bUseMaxDistance=true
    Module.bUseMinDistance=false
    Reset Age On Awaken=true
    Scalability Mode=::NewEnumerator1
    UseLoopDelay=false
   )
   &SpawnRate(
    SpawnRate=&Multiply_Float(
     A=&FloatFromCurve(
      FloatCurve=DataInterface() {propertyValues {Curve {keys [<at 0 0>,<at 0.1 1>,<at 0.2 0>]}}}
      CurveIndex=$Emitter.Age
      Scale Curve=&Multiply_Float(A=25000, B=$User.Spawn Mult)
     )
     B=&FloatFromCurve(
      FloatCurve=DataInterface() {propertyValues {Curve {keys [<at 0 1>,<at 0.2 0>,<at 0.8 0>,<at 1 1>]}}}
      CurveIndex=Linked() {linkedVariable {name=NPC.FortniteNPC.TimeOfDay_0-1, type=NiagaraFloat}}
     )
    )
    Use Spawn Probability=false
   )
  ]
  particleSpawnScript [
   /SetVariables(Particles.RandSpreadCurve:NiagaraFloat=&FloatFromCurve(
     FloatCurve=DataInterface() {propertyValues {Curve {keys [
         {interpMode=RCIM_Cubic, tangentMode=RCTM_User, time=0, value=0.7}
         {interpMode=RCIM_Cubic, tangentMode=RCTM_User, time=0.8, value=0.4}
         {arriveTangent=2.7919, interpMode=RCIM_Cubic, leaveTangent=2.7919, tangentMode=RCTM_User, time=1, value=1}
        ]}}}
     CurveIndex=&RandomRangeFloat(Minimum=0, Maximum=1)
    )) {isSetParameters=true, module=SetVariables_#2}
   &InitializeParticle(
    Color=<rgba 0 1 0.0313 1>
    Lifetime=&RandomRangeFloat(Minimum=2, Maximum=5)
    Sprite Rotation Mode=::NewEnumerator2
    Sprite Size Mode=::NewEnumerator1
    Uniform Sprite Size Max=5
    Uniform Sprite Size Min=2
    UsePositionOffset=false
    Write Material Random=true
   ) {module=InitializeParticle001}
   &ShapeLocation(
    MaskSpawnGroup=false
    Offset Mode=::NewEnumerator0
    Override Seed=false
    Sphere Radius=&Multiply_Float(A=1000, B=$User.Radius Mult)
    UseOwnerScale=true
   )
   /SetVariables(Particles.Position=&Add_Vector(
     A=Linked() {linkedVariable {name=Particles.Position, type=NiagaraPosition}}
     B=&MakeVector(X=0, Y=0, Z=$Particles.Terrain Offset)
    )) {isSetParameters=true, module=SetVariables_#3}
   /SetVariables(Particles.New Vector 3f:Vector3f=<0 0 1>) {isSetParameters=true, module=SetVariables_#4}
   &CameraOffset(Camera Offset Amount=25)
   /SetVariables() {isSetParameters=true, module=SetVariables_#5}
   &InitialMeshOrientation(
    Mesh Orientation Mode=::NewEnumerator1
    Rotation=&RandomRangeVector(Minimum=<0 0 0>, Maximum=<1 1 1>)
    Rotation Coordinate Space=::NewEnumerator2
    Use Rotation Vector=true
   ) {enabled=false}
   /SetVariables(Particles.InitialPos:NiagaraPosition=$Particles.Position) {isSetParameters=true, module=SetVariables_#6}
   /Niagara/Modules/Math/PlaceParticlesOnDepthBuffer_GPU(Max Z Depth=5000, Module.Use Max Z Depth=true)
   /Game/Effects/Niagara/Modules/Spawn/FN_Spawn_GBuffer_Color_WithinRangeV2()
   &KillParticles(Kill Particles=&SetBoolByFloatComparison(
     A=&VectorLength(Vector=&Subtract_Vector(A=$Particles.InitialPos, B=Linked() {linkedVariable {name=Particles.Position, type=NiagaraPosition}}))
     B=500
    )) {module=KillParticles001}
   &KillParticles(Kill Particles=&Boolean_NotOperation(Bool=$Output.PlaceParticlesOnDepthBuffer_GPU.ValidSample)) {module=KillParticles002}
   &CameraOffset(Camera Offset Amount=25) {module=CameraOffset001}
  ]
  particleUpdateScript [
   /Niagara/Modules/Update/Forces/PointAttractionForce(
    Attraction Radius=250
    AttractionStrength:NiagaraFloat=&Multiply_Float(
     A=&FloatFromCurve(FloatCurve=DataInterface() {propertyValues {Curve {keys [<at 0 1>,<at 0.25 0>]}}}, Scale Curve=-150)
     B=1
    )
    AttractorPosition:Vector3f=&Add_Vector(
     A=$Emitter.CameraPosition
     B=&Add_Vector(
      A=&Add_Vector(
       A=&Multiply_VectorByFloat(Vector=$Particles.New Vector 3f, Float=-130)
       B=&Multiply_VectorByFloat(Vector=$Emitter.CameraForward, Float=250)
      )
      B=&Multiply_VectorByFloat(Vector=$Emitter.CameraRight, Float=-50)
     )
    )
    Falloff Exponent=1
    Module.Kill Within Radius=false
    Module.Use Falloff=false
   )
   /SetVariables(Particles.Normalized Velocity:Vector3f=&NormalizeVector(Vector To Normalize=$Particles.Velocity)) {isSetParameters=true, module=SetVariables_#7}
   /SetVariables(Particles.Is Active:NiagaraBool=&Boolean_OrOperation(
     Bool1=$Particles.Is Active
     Bool2=&SetBoolByFloatComparison(A=&MakeFloatFromVector(Vector=$Particles.Normalized Velocity, Channel=::NewEnumerator0), B=0.1)
    )) {isSetParameters=true}
   &CurlNoiseForce(
    Noise Frequency=&RandomRangeFloat(Minimum=5, Maximum=8)
    Noise Strength=&MakeCustomFloatFromBool(Bool=$Particles.Is Active, True Float=&RandomRangeFloat(Minimum=100, Maximum=500), False Float=0)
    Pan Noise=false
   )
   &GravityForce(Gravity=&Multiply_VectorByFloat(
     Vector=<0 0 1>
     Float=&MakeCustomFloatFromBool(Bool=$Particles.Is Active, True Float=&RandomRangeFloat(Minimum=5, Maximum=10), False Float=0)
    ))
   &CurlNoiseForce(
    Noise Frequency=&RandomRangeFloat(Minimum=2, Maximum=3)
    Noise Strength=&MakeCustomFloatFromBool(Bool=$Particles.Is Active, True Float=&RandomRangeFloat(Minimum=5000, Maximum=25000), False Float=0)
    Pan Noise=true
    Pan Noise Field=<0 0 1>
   ) {module=CurlNoiseForce001}
   &SolveForcesAndVelocity(Clamp Velocity=true, Limit Acceleration=false, Speed Limit=100)
   &ParticleState(DeltaTime:NiagaraFloat=&MakeCustomFloatFromBool(
     Bool=&SetBoolByFloatComparison(A=&VectorLength(Vector=$Particles.Velocity), B=1)
     True Float=$Engine.DeltaTime
     False Float=0
    ))
   &KillParticles(Kill Particles=&self:ScratchDynamicInput_01())
   &SolveForcesAndVelocity(Clamp Velocity=true, Limit Acceleration=false, Speed Limit=100) {module=SolveForcesAndVelocity001}
   &ScaleColor(
    Scale Alpha=&FloatFromCurve(
     FloatCurve=DataInterface() {propertyValues {Curve {keys [{interpMode=RCIM_Cubic, tangentMode=RCTM_User, time=0, value=1}, {arriveTangent=1.8182, time=0.35, value=1}, <at 0.55 1>, <at 1 0>]}}}
     Scale Curve=&FloatFromCurve(
      FloatCurve=DataInterface() {propertyValues {Curve {keys [<at 0 1>,<at 0.15 0>,<at 0.8 0>,<at 1 1>]}}}
      CurveIndex=Linked() {linkedVariable {name=NPC.FortniteNPC.TimeOfDay_0-1, type=NiagaraFloat}}
      Scale Curve=5
     )
    )
    Scale RGB=&Multiply_VectorByFloat(Vector=<25 25 25>, Float=1)
    ScaleA=true
    ScaleRGB=true
   )
   &ScaleColor(
    Scale Alpha=&Multiply_Float(
     A=&Waveform(
      "Waveform [1]"=::NewEnumerator3
      "[1] Frequency"=&RandomRangeFloat(Minimum=0.1, Maximum=0.2)
      "[1] Amplitude Min/Max"=<0 1>
      "[1] Exponent"=2
     )
     B=&Waveform(
      "Waveform [1]"=::NewEnumerator0
      "[1] Frequency"=&RandomRangeFloat(Minimum=0.1, Maximum=-0.75)
      "[1] Amplitude Min/Max"=<-1 1>
     )
    )
    ScaleA=true
    ScaleRGB=true
   ) {module=ScaleColor001}
   &DynamicMaterialParameters(
    Index 0 Param 1=1
    Index 0 Param 2=5
    Module.Index0Param0Write=true
    Module.Index0Param1Write=true
    Module.Index0Param2Write=true
    Module.Index0Param3Write=true
   )
   /Niagara/Modules/Update/Camera/MaintainInCameraParticleScale(Render Target Width=2)
   &ScaleSpriteSize(
    Uniform Curve Index=$Engine.Owner.LODDistance
    Uniform Curve Sprite Scale=DataInterface() {propertyValues {Curve {keys [<at 0 0.25>,<at 100 1>,<at 6000 1>,<at 8000 0>]}}}
   )
  ]
  rendererClasses [&NiagaraSpriteRendererProperties]
  renderers [{rendererClass=&NiagaraSpriteRendererProperties}]
 }
 Fireflies001 {
  simTarget=GPUComputeSim
  emitterUpdateScript [
   &EmitterState(
    Enable Distance Culling=true
    Life Cycle Mode=::NewEnumerator1
    Loop Delay=&RandomRangeFloat(Minimum=1, Maximum=2)
    Loop Duration=&RandomRangeFloat(Minimum=2, Maximum=4)
    Max Distance Response=::SleepAndClearParticles
    Module.bUseMaxDistance=true
    Module.bUseMinDistance=false
    Recalculate Duration Each Loop=true
    Reset Age On Awaken=true
    Scalability Mode=::NewEnumerator1
    UseLoopDelay=true
   )
   &SpawnRate(
    SpawnRate=&Multiply_Float(
     A=&FloatFromCurve(
      FloatCurve=DataInterface() {propertyValues {Curve {keys [<at 0 0>,<at 0.25 1>,<at 0.5 0>]}}}
      CurveIndex=$System.NormalizedLoopAge
      Scale Curve=&Multiply_Float(A=500, B=$User.Spawn Mult)
     )
     B=&FloatFromCurve(
      FloatCurve=DataInterface() {propertyValues {Curve {keys [<at 0 1>,<at 0.2 0>,<at 0.8 0>,<at 1 1>]}}}
      CurveIndex=Linked() {linkedVariable {name=NPC.FortniteNPC.TimeOfDay_0-1, type=NiagaraFloat}}
     )
    )
    Use Spawn Probability=false
   )
  ]
  particleSpawnScript [
   /SetVariables(Particles.RandSpreadCurve:NiagaraFloat=&FloatFromCurve(
     FloatCurve=DataInterface() {propertyValues {Curve {keys [
         {interpMode=RCIM_Cubic, tangentMode=RCTM_User, time=0, value=0.7}
         {interpMode=RCIM_Cubic, tangentMode=RCTM_User, time=0.8, value=0.4}
         {arriveTangent=2.7919, interpMode=RCIM_Cubic, leaveTangent=2.7919, tangentMode=RCTM_User, time=1, value=1}
        ]}}}
     CurveIndex=&RandomRangeFloat(Minimum=0, Maximum=1)
    )) {isSetParameters=true, module=SetVariables_#2}
   &InitializeParticle(
    Color=<rgba 0 1 0.0313 1>
    Lifetime=&RandomRangeFloat(Minimum=1, Maximum=2)
    Sprite Rotation Mode=::NewEnumerator2
    Sprite Size Mode=::NewEnumerator1
    Uniform Sprite Size Max=5
    Uniform Sprite Size Min=2
    UsePositionOffset=false
    Write Material Random=true
   ) {module=InitializeParticle001}
   &ShapeLocation(
    MaskSpawnGroup=false
    Offset Mode=::NewEnumerator0
    Override Seed=false
    Sphere Radius=&Multiply_Float(A=1000, B=$User.Radius Mult)
    UseOwnerScale=true
   )
   /SetVariables(Particles.Position=&Add_Vector(
     A=Linked() {linkedVariable {name=Particles.Position, type=NiagaraPosition}}
     B=&MakeVector(X=0, Y=0, Z=$Particles.Terrain Offset)
    )) {isSetParameters=true, module=SetVariables_#3}
   /SetVariables(Particles.New Vector 3f:Vector3f=<0 0 1>) {isSetParameters=true, module=SetVariables_#4}
   &CameraOffset(Camera Offset Amount=25)
   /SetVariables() {isSetParameters=true, module=SetVariables_#5}
   &InitialMeshOrientation(
    Mesh Orientation Mode=::NewEnumerator1
    Rotation=&RandomRangeVector(Minimum=<0 0 0>, Maximum=<1 1 1>)
    Rotation Coordinate Space=::NewEnumerator2
    Use Rotation Vector=true
   ) {enabled=false}
   /SetVariables(Particles.InitialPos:NiagaraPosition=$Particles.Position) {isSetParameters=true, module=SetVariables_#6}
   &CameraOffset(Camera Offset Amount=25) {module=CameraOffset001}
  ]
  particleUpdateScript [
   &PointAttractionForce(
    Attraction Radius=250
    AttractionStrength:NiagaraFloat=&Multiply_Float(
     A=&FloatFromCurve(FloatCurve=DataInterface() {propertyValues {Curve {keys [<at 0 1>,<at 0.25 0>]}}}, Scale Curve=-150)
     B=1
    )
    AttractorPosition:Vector3f=&Add_Vector(
     A=$Emitter.CameraPosition
     B=&Add_Vector(
      A=&Add_Vector(
       A=&Multiply_VectorByFloat(Vector=$Particles.New Vector 3f, Float=-130)
       B=&Multiply_VectorByFloat(Vector=$Emitter.CameraForward, Float=250)
      )
      B=&Multiply_VectorByFloat(Vector=$Emitter.CameraRight, Float=-50)
     )
    )
    Falloff Exponent=1
    Module.Kill Within Radius=false
    Module.Use Falloff=false
   )
   /SetVariables(Particles.Normalized Velocity:Vector3f=&NormalizeVector(Vector To Normalize=$Particles.Velocity)) {isSetParameters=true, module=SetVariables_#7}
   &CurlNoiseForce(
    Noise Frequency=&RandomRangeFloat(Minimum=1, Maximum=25)
    Noise Strength=&MakeCustomFloatFromBool(Bool=true, True Float=&RandomRangeFloat(Minimum=100, Maximum=500), False Float=0)
    Pan Noise=false
   )
   &GravityForce(Gravity=&Multiply_VectorByFloat(
     Vector=<0 0 1>
     Float=&MakeCustomFloatFromBool(Bool=$Particles.Is Active, True Float=&RandomRangeFloat(Minimum=5, Maximum=10), False Float=0)
    ))
   &CurlNoiseForce(
    Noise Frequency=&RandomRangeFloat(Minimum=2, Maximum=3)
    Noise Strength=&MakeCustomFloatFromBool(Bool=$Particles.Is Active, True Float=&RandomRangeFloat(Minimum=5000, Maximum=25000), False Float=0)
    Pan Noise=true
    Pan Noise Field=<0 0 1>
   ) {module=CurlNoiseForce001}
   &SolveForcesAndVelocity(Clamp Velocity=true, Limit Acceleration=false, Speed Limit=25)
   &ParticleState(DeltaTime:NiagaraFloat=&MakeCustomFloatFromBool(
     Bool=&SetBoolByFloatComparison(A=&VectorLength(Vector=$Particles.Velocity), B=1)
     True Float=$Engine.DeltaTime
     False Float=0
    ))
   &KillParticles(Kill Particles=&self:ScratchDynamicInput_01())
   &SolveForcesAndVelocity(Clamp Velocity=true, Limit Acceleration=false, Speed Limit=100) {module=SolveForcesAndVelocity001}
   &ScaleColor(
    Scale Alpha=&FloatFromCurve(
     FloatCurve=DataInterface() {propertyValues {Curve {keys [{interpMode=RCIM_Cubic, tangentMode=RCTM_User, time=0, value=1}, {arriveTangent=1.8182, time=0.35, value=1}, <at 0.55 1>, <at 1 0>]}}}
     Scale Curve=&FloatFromCurve(
      FloatCurve=DataInterface() {propertyValues {Curve {keys [<at 0 1>,<at 0.15 0>,<at 0.8 0>,<at 1 1>]}}}
      CurveIndex=Linked() {linkedVariable {name=NPC.FortniteNPC.TimeOfDay_0-1, type=NiagaraFloat}}
      Scale Curve=5
     )
    )
    Scale RGB=&Multiply_VectorByFloat(Vector=<25 25 25>, Float=1)
    ScaleA=true
    ScaleRGB=true
   )
   &ScaleColor(
    Scale Alpha=&Multiply_Float(
     A=&Waveform(
      "Waveform [1]"=::NewEnumerator3
      "[1] Frequency"=&RandomRangeFloat(Minimum=0.1, Maximum=0.2)
      "[1] Amplitude Min/Max"=<0 1>
      "[1] Exponent"=2
     )
     B=&Waveform(
      "Waveform [1]"=::NewEnumerator0
      "[1] Frequency"=&RandomRangeFloat(Minimum=0.1, Maximum=-0.75)
      "[1] Amplitude Min/Max"=<-1 1>
     )
    )
    ScaleA=true
    ScaleRGB=true
   ) {module=ScaleColor001}
   &DynamicMaterialParameters(
    Index 0 Param 1=1
    Index 0 Param 2=5
    Module.Index0Param0Write=true
    Module.Index0Param1Write=true
    Module.Index0Param2Write=true
    Module.Index0Param3Write=true
   )
   &MaintainInCameraParticleScale(Render Target Width=2)
   &ScaleSpriteSize(
    Uniform Curve Index=$Engine.Owner.LODDistance
    Uniform Curve Sprite Scale=DataInterface() {propertyValues {Curve {keys [<at 0 0.25>,<at 100 1>,<at 6000 1>,<at 8000 0>]}}}
   )
  ]
  rendererClasses [&NiagaraSpriteRendererProperties]
  renderers [{rendererClass=&NiagaraSpriteRendererProperties}]
 }
}
summary {
 systemName=NS_Env_Forest_FireFlies_Volume
 emitters [
  {emitterName=Fireflies, simTarget=GPUComputeSim, rendererClasses [&NiagaraSpriteRendererProperties]}
  {emitterName=Fireflies001, simTarget=GPUComputeSim, rendererClasses [&NiagaraSpriteRendererProperties]}
 ]
 userVariables [
  User.Radius Mult {type=NiagaraFloat, defaultValue=1, description=""}
  User.Spawn Mult {type=NiagaraFloat, defaultValue=0.5, description=""}
 ]
}
