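# Niagara GPU sprite system of green ground leaves that are stirred up by a repulsion point derived from the camera.
# Leaves spawn within a user-tuned, curve-weighted radius and snap to sampled landscape height, with time-of-day tint and distance culling.
# Negative point attraction pushes them, and an active flag from terrain-height comparison gates wind, curl noise, drag and a landscape-normal to camera-facing orientation switch.
# Particle aging is gated by velocity.
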NiagaraSystem /Game/Effects/Fort_Effects/Effects/Environment/Forest/NS_Env_LeavesOnGround_Interact_Green
dependencies {
 usedDataInterfaces [&NiagaraDataInterfaceCamera]
 usedDynamicInputs [&SimulationPosition]
 usedModules [&SystemState]
}
emitters {InteractiveConfetti {
  simTarget=GPUComputeSim
  emitterUpdateScript [
   /SetVariables() {isSetParameters=true}
   /SetVariables(
    Emitter.CameraForward:Vector3f=&ReturnCameraVector(Camera Property=::NewEnumerator1, Coordinate Space=::Simulation, CameraQuery=$Emitter.CameraRead)
    Emitter.CameraPosition:Vector3f=&ReturnCameraVector(Camera Property=::NewEnumerator0, Coordinate Space=::World, CameraQuery=$Emitter.CameraRead)
    Emitter.CameraRight:Vector3f=&ReturnCameraVector(Camera Property=::NewEnumerator3, Coordinate Space=::Simulation, CameraQuery=$Emitter.CameraRead)
    Emitter.CameraUp:Vector3f=&ReturnCameraVector(Camera Property=::NewEnumerator2, Coordinate Space=::Simulation, CameraQuery=$Emitter.CameraRead)
   ) {isSetParameters=true}
   &EmitterState(
    Enable Distance Culling=true
    Life Cycle Mode=::NewEnumerator1
    Loop Behavior=::NewEnumerator1
    Loop Duration=0.25
    Max Distance Response=::SleepAndClearParticles
    MaxDistance=8000
    Module.bApplySpawnCountScaleByDistance=false
    Module.bUseMaxDistance=true
    Module.bUseMinDistance=false
    Reset Age On Awaken=true
    Scalability Mode=::NewEnumerator1
    Scale Spawn Count=true
    Spawn Count Scale By Distance=DataInterface() {propertyValues {Curve {keys [<at 0 0>,<at 0.3 0.95>,<at 0.75 1>,<at 1 0>]}}}
    UseLoopDelay=false
   )
   &SpawnRate(SpawnRate=$User.SpawnRate, Use Spawn Probability=false)
  ]
  particleSpawnScript [
   &InitializeParticle(
    Color=&RandomRangeLinearColor(Minimum=<rgba 0.4742 0.9844 0.4255 1>, Maximum=<rgba 0.3389 0.4635 0.0748 1>)
    Lifetime=&RandomRangeFloat(Minimum=3, Maximum=5)
    Mass=&RandomRangeFloat(Minimum=3, Maximum=6)
    Mass Mode=::NewEnumerator1
    Mesh Scale Mode=::NewEnumerator4
    Sprite Rotation Angle=&RandomRangeFloat(Minimum=0, Maximum=360)
    Sprite Rotation Mode=::NewEnumerator0
    Sprite Size=&Vector2DFromFloat(Value=&RandomRangeFloat(Minimum=15, Maximum=25))
    Sprite Size Mode=::NewEnumerator4
    UsePositionOffset=false
    Write Material Random=true
   ) {module=InitializeParticle001}
   /SetVariables(Particles.Terrain Offset=5) {isSetParameters=true}
   /Niagara/Modules/Spawn/Location/SphereLocation(
    MaskSpawnGroup=false
    Module.Override Randomness=false
    Sphere Radius=&FloatFromCurve(
     FloatCurve=$User.DistributionWeight
     CurveIndex=&RandomRangeFloat(Minimum=0, Maximum=$User.MaxWeight)
     Scale Curve=$User.RadiusSpread
    )
    Surface Only=false
   )
   &self:Sample_Landscape() {module=Sample_Landscape}
   /SetVariables(Particles.Position=&Add_Vector(
     A=Linked() {linkedVariable {name=Particles.Position, type=NiagaraPosition}}
     B=&MakeVector(X=0, Y=0, Z=$Particles.Terrain Offset)
    )) {isSetParameters=true}
   &AddVelocity()
   &CameraOffset(Camera Offset Amount=2)
   &SubUVAnimation(End Frame=36, SubUV Animation Mode:ENiagara_SubUVLookupMode=::NewEnumerator2) {module=SubUVAnimation001}
   /SetVariables(Particles.SubImageIndex=&RandomRangeFloat(Minimum=0, Maximum=35)) {isSetParameters=true}
   /SetVariables() {isSetParameters=true}
  ]
  particleUpdateScript [
   &self:Sample_Landscape_UPdate(Bias:NiagaraFloat=$Particles.Terrain Offset, New Landscape Sample:NiagaraDataInterfaceLandscape=$User.Landscape) {module=Sample_Landscape_UPdate}
   &ScaleColor(
    Scale Alpha=&FloatFromCurve(FloatCurve=DataInterface() {propertyValues {Curve {keys [<at 0.9 5>,<at 1 0>]}}})
    Scale RGB=&MakeVectorFromLinearColorRGB(Color=&MultiplyLinearColor(
      Linear Color=<rgba 0.4198 0.5 0.2292 1>
      Scale Factor=&MakeVector4_FromFloat(Float=&FNTimeOfDayScalar(Night Scale=0.5, Morning Scale=0.5, Day Scale=0.6, Evening Scale=0.5))
     ))
    ScaleA=true
    ScaleRGB=true
   )
   /SetVariables(Particles.Has Collided With Terrain:NiagaraBool=&SetBoolByFloatComparison(
     A=&Add_Float(A=$Output.Sample_Landscape_UPdate.Terrain Height, B=20)
     B=&MakeFloatFromVector(Vector=Linked() {linkedVariable {name=Particles.Position, type=NiagaraPosition}}, Channel=::NewEnumerator2)
    )) {isSetParameters=true}
   /Niagara/Modules/Update/Forces/PointAttractionForce(
    AttractionStrength=-85
    AttractorPosition:Vector3f=&Add_Vector(
     A=$Emitter.CameraPosition
     B=&Add_Vector(
      A=&Add_Vector(
       A=&Multiply_VectorByFloat(Vector=$Particles.Landscape Normal, Float=-200)
       B=&Multiply_VectorByFloat(Vector=$Emitter.CameraForward, Float=300)
      )
      B=&Multiply_VectorByFloat(Vector=$Emitter.CameraRight, Float=-50)
     )
    )
    Falloff Exponent=1
    Module.Kill Within Radius=false
    Module.Use Falloff=false
   )
   /SetVariables(Particles.Normalized Velocity:Vector3f=&NormalizeVector(Vector To Normalize=$Particles.Velocity)) {isSetParameters=true}
   /SetVariables(Particles.Is Active:NiagaraBool=&Boolean_OrOperation(
     Bool1=$Particles.Is Active
     Bool2=&SetBoolByFloatComparison(
      A=&MakeFloatFromVector(Vector=Linked() {linkedVariable {name=Particles.Position, type=NiagaraPosition}}, Channel=::NewEnumerator2)
      B=$Output.Sample_Landscape_UPdate.Terrain Height
      Comparison Type=::NewEnumerator5
     )
    )) {isSetParameters=true}
   &WindForce(
    Air Resistance=1
    Wind Speed=&Multiply_VectorByFloat(
     Vector=&RandomRangeVector(Minimum=<250 -250 250>, Maximum=<600 250 500>)
     Float=&MakeCustomFloatFromBool(
      Bool=$Particles.Is Active
      True Float=&FloatFromCurve(FloatCurve=DataInterface() {propertyValues {Curve {keys [<at 0 1>,<at 1 0>]}}})
      False Float=0
     )
    )
    Wind Speed Scale=&MakeCustomFloatFromBool(Bool=$Particles.Has Collided With Terrain, True Float=0, False Float=2)
   )
   &CurlNoiseForce(
    Noise Frequency=&RandomRangeFloat(Minimum=2, Maximum=10)
    Noise Strength=&MakeCustomFloatFromBool(
     Bool=$Particles.Is Active
     True Float=&FloatFromCurve(
      FloatCurve=DataInterface() {propertyValues {Curve {keys [
          {interpMode=RCIM_Cubic, tangentMode=RCTM_User, time=0.25, value=1}
          {arriveTangent=-3.7637, interpMode=RCIM_Cubic, leaveTangent=-3.7637, tangentMode=RCTM_User, time=0.8, value=0.4}
          <at 1 0>
         ]}}}
      Scale Curve=1500
     )
     False Float=0
    )
    Pan Noise=false
   ) {module=CurlNoiseForce001}
   &Drag(
    Drag=&MakeCustomFloatFromBool(
     Bool=$Particles.Is Active
     True Float=&FloatFromCurve(FloatCurve=DataInterface() {propertyValues {Curve {keys [<at 0.1 0>,<at 0.25 2>,<at 0.35 0.3>,<at 0.7 0.1>]}}}, Scale Curve=0.25)
     False Float=0
    )
    Use Linear Drag=true
    Use Rotational Drag=false
   )
   &GravityForce(Gravity=&MakeVector(
     X=0
     Y=0
     Z=&FloatFromCurve(FloatCurve=DataInterface() {propertyValues {Curve {keys [<at 0 0 RCIM_Cubic>,<at 0.2 -1 RCIM_Cubic>]}}}, Scale Curve=600)
    )) {module=GravityForce001}
   /SetVariables(Particles.Sprite Orientation:Vector3f=&Lerp_Vector(
     A=$Particles.Landscape Normal
     B=&ReturnCameraVector(Camera Property=::NewEnumerator4, Coordinate Space=::Simulation, Origin Position=$Particles.Position, CameraQuery=$Emitter.CameraRead)
     Alpha=&MakeCustomFloatFromBool(Bool=$Particles.Is Active, True Float=1, False Float=0)
    )) {isSetParameters=true}
   &SolveForcesAndVelocity(Clamp Velocity=true, Limit Acceleration=false, Speed Limit=750)
   &self:ScratchModule_03(Terrain Height:NiagaraFloat=$Output.Sample_Landscape_UPdate.Terrain Height, Threshold=20) {module=ScratchModule_03}
   &SubUVAnimation(End Frame=64, SubUV Animation Mode:ENiagara_SubUVLookupMode=::NewEnumerator1, SubUV Loop Count=4)
   &ParticleState(DeltaTime:NiagaraFloat=&MakeCustomFloatFromBool(
     Bool=&SetBoolByFloatComparison(A=&VectorLength(Vector=$Particles.Velocity), B=1)
     True Float=$Engine.DeltaTime
     False Float=0
    ))
   &ScaleSpriteSize(Uniform Curve Sprite Scale=DataInterface() {propertyValues {Curve {keys [{tangentMode=RCTM_User, time=0.8, value=1}, {tangentMode=RCTM_User, time=1, value=0}]}}})
   &SpriteRotationRate(Rotation Rate=3)
   &DynamicMaterialParameters(
    Index 0 Param 1=0.04
    Index 0 Param 2=&RandomRangeFloat(Minimum=0.02, Maximum=0.08)
    Module.Param0WriteEnabled=true
    Module.Param1WriteEnabled=true
    Module.Param2WriteEnabled=true
    Module.Param3WriteEnabled=true
   )
  ]
  rendererClasses [&NiagaraSpriteRendererProperties]
  renderers [{rendererClass=&NiagaraSpriteRendererProperties}]
 }}
summary {
 systemName=NS_Env_LeavesOnGround_Interact_Green
 emitters [{emitterName=InteractiveConfetti, simTarget=GPUComputeSim, rendererClasses [&NiagaraSpriteRendererProperties]}]
 userVariables [
  User.Alpha_AttractionStrength {type=NiagaraFloat, defaultValue=0, description=""}
  User.DistributionWeight {
   type=NiagaraDataInterfaceCurve
   description=""
   defaultValue {
    struct=NiagaraExt_VariableValue_DataInterface
    value {
     dataInterface=&self:NiagaraDataInterfaceCurve_0
     dataInterfaceClass=&NiagaraDataInterfaceCurve
    }
   }
  }
  User.Landscape {
   type=NiagaraDataInterfaceLandscape
   description=""
   defaultValue {
    struct=NiagaraExt_VariableValue_DataInterface
    value {
     dataInterface=&self:NiagaraDataInterfaceLandscape_0
     dataInterfaceClass=&NiagaraDataInterfaceLandscape
    }
   }
  }
  User.MaxWeight {type=NiagaraFloat, defaultValue=1, description=""}
  User.Player Position {type=Vector3f, description="", defaultValue {struct=Vector3f, value=<0 0 0>}}
  User.RadiusSpread {type=NiagaraFloat, defaultValue=1500, description=""}
  User.SpawnRate {type=NiagaraFloat, defaultValue=10000, description=""}
 ]
}
